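# Physics asset for the coat, belt buckle and front/back chains on the male medium Lars player body.
# Capsules on pelvis, thighs and calves, plus two large root spheres, form the body shapes.
# Six four-segment coat strips, the buckle and front chain hang from pelvis-linked MSTR bodies with angular limits; the back chain links to the right thigh.
# Extra constraints cross-link neighbouring rear coat strips, and both chains and right thigh.
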PhysicsAsset /BRCosmetics/Characters/Player/Male/Medium/Bodies/M_MED_Lars/Meshes/M_MED_Lars_Physics_Coat
bodies {
 calf_l [{shapeType=Capsule, length=56.9216, radius=10.5, center=<-31.3035 -0.2388 0.0141>, rotation=<rot 90 0 0>}]
 calf_r [{shapeType=Capsule, length=50, radius=10.5, center=<32.5574 0 0.0001>, rotation=<rot 90 0 0>}]
 dyn_beltBuckle [{
   shapeType=Capsule
   length=4.071
   radius=4.066
   shapeName=dyn_beltBuckle_capsule
   center=<0.6329 0.5254 -1.5585>
   rotation=<rot -0.5934 90.0031 173.9066>
  }]
 dyn_beltBuckle_MSTR [{shapeType=Capsule, length=0.505, radius=0.505, shapeName=dyn_beltBuckle_MSTR_capsule, rotation=<rot 0 -90 0>}]
 dyn_chainBack_01 [{shapeType=Capsule, length=5, radius=2, shapeName=dyn_chainBack_01_capsule, center=<4 0 0>, rotation=<rot 90 0 0>}]
 dyn_chainBack_02 [{shapeType=Capsule, length=5, radius=2, shapeName=dyn_chainBack_02_capsule, center=<2 0 0>, rotation=<rot 90 0 0>}]
 dyn_chainFront_01 [{shapeType=Capsule, length=5.5, radius=2, shapeName=dyn_chainFront_01_capsule, center=<4 0 0>, rotation=<rot 90 0 0>}]
 dyn_chainFront_02 [{shapeType=Capsule, length=5.5, radius=2, shapeName=dyn_chainFront_02_capsule, center=<4 0 0>, rotation=<rot 90 0 0>}]
 dyn_chainFront_03 [{shapeType=Capsule, length=5.5, radius=2, shapeName=dyn_chainFront_03_capsule, center=<4 0 0>, rotation=<rot 90 0 0>}]
 dyn_chainFront_04 [{shapeType=Capsule, length=5.5, radius=2, shapeName=dyn_chainFront_04_capsule, center=<4 0 0>, rotation=<rot 90 0 0>}]
 dyn_chainFront_05 [{shapeType=Capsule, length=5.5, radius=2, shapeName=dyn_chainFront_05_capsule, center=<4 0 0>, rotation=<rot 90 0 0>}]
 dyn_chainFront_06 [{shapeType=Capsule, length=5.5, radius=2, shapeName=dyn_chainFront_06_capsule, center=<4 0 0>, rotation=<rot 90 0 0>}]
 dyn_chainFront_07 [{shapeType=Capsule, length=0, radius=2, shapeName=dyn_chainFront_07_capsule}]
 dyn_chainFront_MSTR [{shapeType=Capsule, length=0.505, radius=0.505, shapeName=dyn_chainFront_MSTR_capsule, rotation=<rot 90 0 0>}]
 dyn_jacket_bk_l_1 [{shapeType=Capsule, length=14, radius=4, shapeName=dyn_jacketSide_01_l_capsule, center=<5 0 3.5>, rotation=<rot 0 0 90>}]
 dyn_jacket_bk_l_2 [{shapeType=Capsule, length=14.0192, radius=8.9808, shapeName=dyn_jacketSide_02_l_capsule, center=<9 0 3>, rotation=<rot 0 0 90>}]
 dyn_jacket_bk_l_3 [{shapeType=Capsule, length=15.7692, radius=8.7308, shapeName=dyn_jacketBack_03_l_capsule, center=<8 0 3>, rotation=<rot 0 0 90>}]
 dyn_jacket_bk_l_4 [{shapeType=Capsule, length=20.7692, radius=9.2308, shapeName=dyn_jacketBack_04_l_capsule, center=<7 0 3>, rotation=<rot 0 0 90>}]
 dyn_jacket_bk_l_MSTR [{shapeType=Capsule, length=0.505, radius=0.505, shapeName=dyn_jacket_bk_l_MSTR_capsule, rotation=<rot 0 -90 0>}]
 dyn_jacket_bk_r_1 [{shapeType=Capsule, length=11.9586, radius=6.0414, shapeName=dyn_jacketSide_01_l_capsule, center=<5 0 3.5>, rotation=<rot 0 0 90>}]
 dyn_jacket_bk_r_2 [{shapeType=Capsule, length=14.0192, radius=8.9808, shapeName=dyn_jacketSide_02_l_capsule, center=<9 0 3>, rotation=<rot 0 0 90>}]
 dyn_jacket_bk_r_3 [{shapeType=Capsule, length=15.8646, radius=10.6354, shapeName=dyn_jacketBack_03_r_capsule, center=<8 0 3>, rotation=<rot 0 0 90>}]
 dyn_jacket_bk_r_4 [{shapeType=Capsule, length=23.8646, radius=11.1354, shapeName=dyn_jacketBack_04_l_capsule, center=<7 0 3>, rotation=<rot 0 0 90>}]
 dyn_jacket_bk_r_MSTR [{shapeType=Capsule, length=0.505, radius=0.505, shapeName=dyn_jacket_bk_r_MSTR_capsule, rotation=<rot 0 -90 0>}]
 dyn_jacket_fr_l_1 [{shapeType=Capsule, length=9.6731, radius=3.5, shapeName=dyn_jacketFront_01_l_capsule, center=<12.2259 0 2.5>, rotation=<rot 90 0 0>}]
 dyn_jacket_fr_l_2 [
  {shapeType=Capsule, length=20, radius=4.5, shapeName=dyn_jacketFront_03_l_capsule, center=<10 -7 2>, rotation=<rot 0 0 90>}
  {
   shapeType=Capsule
   length=16
   radius=1.25
   shapeName=dyn_jacketFront_03_l_capsule
   center=<9.5898 -2.5 0>
   rotation=<rot 86.4659 179.9998 179.9998>
  }
 ]
 dyn_jacket_fr_l_3 [
  {shapeType=Capsule, length=25, radius=4.5, shapeName=dyn_jacketFront_03_l_capsule, center=<8 -6 2>, rotation=<rot 0 0 90>}
  {
   shapeType=Capsule
   length=16
   radius=1.25
   shapeName=dyn_jacketFront_03_l_capsule
   center=<9.5898 0 0.5923>
   rotation=<rot 86.4659 179.9998 179.9998>
  }
 ]
 dyn_jacket_fr_l_4 [{shapeType=Capsule, length=25, radius=4.5, shapeName=dyn_jacketFront_04_l_capsule, center=<4 -6 2>, rotation=<rot 0 -10 90>}]
 dyn_jacket_fr_l_MSTR [{shapeType=Capsule, length=0.505, radius=0.505, shapeName=dyn_jacket_fr_l_MSTR_capsule, rotation=<rot 0 -90 0>}]
 dyn_jacket_fr_r_1 [{shapeType=Capsule, length=8.1087, radius=3.5, shapeName=dyn_jacketFront_01_l_capsule, center=<8 0 3>, rotation=<rot 90 0 0>}]
 dyn_jacket_fr_r_2 [
  {shapeType=Capsule, length=20, radius=4.5, shapeName=dyn_jacketFront_03_l_capsule, center=<10 7 3>, rotation=<rot 0 0 90>}
  {
   shapeType=Capsule
   length=16
   radius=1.25
   shapeName=dyn_jacketFront_03_l_capsule
   center=<9.5898 0 0.5923>
   rotation=<rot 86.4659 179.9998 179.9998>
  }
 ]
 dyn_jacket_fr_r_3 [
  {shapeType=Capsule, length=25, radius=4.5, shapeName=dyn_jacketFront_03_l_capsule, center=<8 6 2.25>, rotation=<rot 0 0 90>}
  {
   shapeType=Capsule
   length=16
   radius=1.25
   shapeName=dyn_jacketFront_03_l_capsule
   center=<9.5898 0 0.5923>
   rotation=<rot 86.4659 179.9998 179.9998>
  }
 ]
 dyn_jacket_fr_r_4 [{shapeType=Capsule, length=25, radius=4.5, shapeName=dyn_jacketFront_04_l_capsule, center=<4 6 1>, rotation=<rot 0 10 90>}]
 dyn_jacket_fr_r_MSTR [{shapeType=Capsule, length=0.505, radius=0.505, shapeName=dyn_jacket_fr_r_MSTR_capsule, rotation=<rot 0 -90 0>}]
 dyn_jacket_mid_l_1 [{shapeType=Capsule, length=14, radius=3.5, shapeName=dyn_jacketSide_01_l_capsule, center=<8 0 2>, rotation=<rot 90 0 0>}]
 dyn_jacket_mid_l_2 [{shapeType=Capsule, length=14.0192, radius=8.5, shapeName=dyn_jacketSide_02_l_capsule, center=<9 0 3>, rotation=<rot 0 0 100>}]
 dyn_jacket_mid_l_3 [{shapeType=Capsule, length=18.5192, radius=8.7308, shapeName=dyn_jacketSide_03_l_capsule, center=<8 0 3.5>, rotation=<rot 0 0 100>}]
 dyn_jacket_mid_l_4 [{shapeType=Capsule, length=22.0192, radius=8.7308, shapeName=dyn_jacketSide_04_r_capsule, center=<6.5 0 3.5>, rotation=<rot 0 -10 100>}]
 dyn_jacket_mid_l_MSTR [{shapeType=Capsule, length=0.505, radius=0.505, shapeName=dyn_jacket_mid_l_MSTR_capsule, rotation=<rot 0 -90 0>}]
 dyn_jacket_mid_r_1 [{shapeType=Capsule, length=12.2293, radius=5.7707, shapeName=dyn_jacketSide_01_l_capsule, center=<5 0 2>, rotation=<rot 0 0 75>}]
 dyn_jacket_mid_r_2 [{shapeType=Capsule, length=14.0192, radius=8.9808, shapeName=dyn_jacketSide_02_l_capsule, center=<9 0 3.25>, rotation=<rot 0 0 80>}]
 dyn_jacket_mid_r_3 [{shapeType=Capsule, length=18.5192, radius=8.7308, shapeName=dyn_jacketSide_03_l_capsule, center=<8 0 3.25>, rotation=<rot 0 0 80>}]
 dyn_jacket_mid_r_4 [{shapeType=Capsule, length=22.0192, radius=8.7308, shapeName=dyn_jacketSide_04_r_capsule, center=<6.5 0 3.25>, rotation=<rot 0 10 80>}]
 dyn_jacket_mid_r_MSTR [{shapeType=Capsule, length=0.505, radius=0.505, shapeName=dyn_jacket_mid_r_MSTR_capsule, rotation=<rot 0 -90 0>}]
 pelvis [
  {shapeType=Capsule, length=11.7621, radius=11.9879, center=<-0.649 0.2801 -0.0002>, rotation=<rot 90 0 -3.3073>}
  {
   shapeType=Capsule
   length=15.3785
   radius=7.8
   shapeName=pelvis_capsule
   center=<-0.0808 -1.75 -4.5558>
   rotation=<rot 80.001 -176.6924 -180>
  }
  {shapeType=Capsule, length=15.6212, radius=7.5, shapeName=pelvis_capsule, center=<-0.2328 -1.75 5.3434>, rotation=<rot 77.8519 3.3074 0>}
 ]
 root [
  {shapeType=Sphere, radius=37, shapeName=root_sphere, center=<0.0002 0 6>}
  {shapeType=Sphere, radius=183.4375, shapeName=root_sphere, center=<0 -8.5 -167>}
 ]
 thigh_l [
  {shapeType=Capsule, length=36.9148, radius=9, center=<-24.7887 0.041 0.6779>, rotation=<rot 88 -167.0926 -169.0593>}
  {
   shapeType=Capsule
   length=34.6544
   radius=6.2296
   shapeName=thigh_l_capsule
   center=<-27.4653 13.9354 4.5479>
   rotation=<rot 88 -167.0926 -169.0593>
  }
 ]
 thigh_r [
  {shapeType=Capsule, length=30, radius=9, center=<21.4597 0.1549 0.7921>, rotation=<rot 92 12.9074 10.9407>}
  {
   shapeType=Capsule
   length=36.7483
   radius=5.2707
   shapeName=thigh_r_capsule
   center=<27.0585 -12.231 -3.1374>
   rotation=<rot 88 -167.0926 -169.0593>
  }
 ]
}
constraints [
 {
  bone1Name=thigh_l
  bone2Name=pelvis
  swing1LimitDegrees=45
  swing1Motion=Free
  swing2LimitDegrees=45
  swing2Motion=Free
  twistLimitDegrees=45
  twistMotion=Free
 }
 {
  bone1Name=thigh_r
  bone2Name=pelvis
  swing1LimitDegrees=45
  swing1Motion=Free
  swing2LimitDegrees=45
  swing2Motion=Free
  twistLimitDegrees=45
  twistMotion=Free
 }
 {
  bone1Name=calf_l
  bone2Name=thigh_l
  swing1LimitDegrees=45
  swing1Motion=Free
  swing2LimitDegrees=45
  swing2Motion=Free
  twistLimitDegrees=45
  twistMotion=Free
 }
 {
  bone1Name=calf_r
  bone2Name=thigh_r
  swing1LimitDegrees=45
  swing1Motion=Free
  swing2LimitDegrees=45
  swing2Motion=Free
  twistLimitDegrees=45
  twistMotion=Free
 }
 {
  bone1Name=pelvis
  bone2Name=root
  swing1LimitDegrees=45
  swing1Motion=Free
  swing2LimitDegrees=45
  swing2Motion=Free
  twistLimitDegrees=45
  twistMotion=Free
 }
 {bone1Name=dyn_beltBuckle, bone2Name=dyn_beltBuckle_MSTR, swing1LimitDegrees=5, swing2LimitDegrees=5, twistLimitDegrees=1}
 {
  bone1Name=thigh_l
  bone2Name=thigh_r
  swing1LimitDegrees=45
  swing1Motion=Free
  swing2LimitDegrees=45
  swing2Motion=Free
  twistLimitDegrees=45
  twistMotion=Free
 }
 {
  bone1Name=dyn_chainFront_MSTR
  bone2Name=pelvis
  swing1LimitDegrees=45
  swing1Motion=Free
  swing2LimitDegrees=45
  swing2Motion=Free
  twistLimitDegrees=45
  twistMotion=Free
 }
 {bone1Name=dyn_chainFront_01, bone2Name=dyn_chainFront_MSTR, swing1LimitDegrees=10, swing2LimitDegrees=10, twistLimitDegrees=10}
 {bone1Name=dyn_chainFront_02, bone2Name=dyn_chainFront_01, swing1LimitDegrees=15, swing2LimitDegrees=15, twistLimitDegrees=5}
 {bone1Name=dyn_chainFront_03, bone2Name=dyn_chainFront_02, swing1LimitDegrees=15, swing2LimitDegrees=15, twistLimitDegrees=7}
 {bone1Name=dyn_chainFront_04, bone2Name=dyn_chainFront_03, swing1LimitDegrees=15, swing2LimitDegrees=15, twistLimitDegrees=10}
 {bone1Name=dyn_chainFront_05, bone2Name=dyn_chainFront_04, swing1LimitDegrees=15, swing2LimitDegrees=15, twistLimitDegrees=10}
 {bone1Name=dyn_chainFront_06, bone2Name=dyn_chainFront_05, swing1LimitDegrees=10, swing2LimitDegrees=10, twistLimitDegrees=10}
 {bone1Name=dyn_chainFront_07, bone2Name=dyn_chainFront_06, swing1LimitDegrees=10, swing2LimitDegrees=10, twistLimitDegrees=10}
 {
  bone1Name=dyn_chainBack_01
  bone2Name=thigh_r
  swing1LimitDegrees=45
  swing1Motion=Free
  swing2LimitDegrees=45
  swing2Motion=Free
  twistLimitDegrees=45
  twistMotion=Free
 }
 {
  bone1Name=dyn_chainBack_02
  bone2Name=dyn_chainBack_01
  swing1LimitDegrees=45
  swing1Motion=Free
  swing2LimitDegrees=45
  swing2Motion=Free
  twistLimitDegrees=45
  twistMotion=Free
 }
 {bone1Name=thigh_r, bone2Name=dyn_chainFront_04, swing1LimitDegrees=2, swing2LimitDegrees=2, twistLimitDegrees=2}
 {
  bone1Name=dyn_chainBack_02
  bone2Name=dyn_chainFront_05
  swing1LimitDegrees=45
  swing1Motion=Free
  swing2LimitDegrees=45
  swing2Motion=Free
  twistLimitDegrees=45
  twistMotion=Free
 }
 {bone1Name=thigh_r, bone2Name=dyn_chainFront_05, swing1LimitDegrees=2, swing2LimitDegrees=2, twistLimitDegrees=2}
 {
  bone1Name=dyn_chainBack_01
  bone2Name=dyn_chainFront_06
  swing1LimitDegrees=45
  swing1Motion=Free
  swing2LimitDegrees=45
  swing2Motion=Free
  twistLimitDegrees=45
  twistMotion=Free
 }
 {
  bone1Name=dyn_chainBack_01
  bone2Name=dyn_chainFront_07
  swing1LimitDegrees=45
  swing1Motion=Free
  swing2LimitDegrees=45
  swing2Motion=Free
  twistLimitDegrees=45
  twistMotion=Free
 }
 {
  bone1Name=dyn_beltBuckle_MSTR
  bone2Name=pelvis
  swing1LimitDegrees=45
  swing1Motion=Free
  swing2LimitDegrees=45
  swing2Motion=Free
  twistLimitDegrees=45
  twistMotion=Free
 }
 {
  bone1Name=dyn_chainBack_02
  bone2Name=dyn_chainFront_06
  swing1LimitDegrees=45
  swing1Motion=Free
  swing2LimitDegrees=45
  swing2Motion=Free
  twistLimitDegrees=45
  twistMotion=Free
 }
 {
  bone1Name=dyn_jacket_fr_l_MSTR
  bone2Name=pelvis
  swing1LimitDegrees=45
  swing1Motion=Free
  swing2LimitDegrees=45
  swing2Motion=Free
  twistLimitDegrees=45
  twistMotion=Free
 }
 {bone1Name=dyn_jacket_fr_l_1, bone2Name=dyn_jacket_fr_l_MSTR, swing1LimitDegrees=10, swing2LimitDegrees=45, twistLimitDegrees=5}
 {bone1Name=dyn_jacket_fr_l_2, bone2Name=dyn_jacket_fr_l_1, swing1LimitDegrees=10, swing2LimitDegrees=45, twistLimitDegrees=5}
 {bone1Name=dyn_jacket_fr_l_3, bone2Name=dyn_jacket_fr_l_2, swing1LimitDegrees=10, swing2LimitDegrees=45, twistLimitDegrees=5}
 {bone1Name=dyn_jacket_fr_l_4, bone2Name=dyn_jacket_fr_l_3, swing1LimitDegrees=10, swing2LimitDegrees=45, twistLimitDegrees=5}
 {
  bone1Name=dyn_jacket_mid_l_MSTR
  bone2Name=pelvis
  swing1LimitDegrees=45
  swing1Motion=Free
  swing2LimitDegrees=45
  swing2Motion=Free
  twistLimitDegrees=45
  twistMotion=Free
 }
 {bone1Name=dyn_jacket_mid_l_1, bone2Name=dyn_jacket_mid_l_MSTR, swing1LimitDegrees=30, swing2LimitDegrees=45, twistLimitDegrees=20}
 {bone1Name=dyn_jacket_mid_l_2, bone2Name=dyn_jacket_mid_l_1, swing1LimitDegrees=30, swing2LimitDegrees=45, twistLimitDegrees=20}
 {bone1Name=dyn_jacket_mid_l_3, bone2Name=dyn_jacket_mid_l_2, swing1LimitDegrees=30, swing2LimitDegrees=45, twistLimitDegrees=20}
 {bone1Name=dyn_jacket_mid_l_4, bone2Name=dyn_jacket_mid_l_3, swing1LimitDegrees=30, swing2LimitDegrees=45, twistLimitDegrees=20}
 {
  bone1Name=dyn_jacket_bk_l_MSTR
  bone2Name=pelvis
  swing1LimitDegrees=45
  swing1Motion=Free
  swing2LimitDegrees=45
  swing2Motion=Free
  twistLimitDegrees=45
  twistMotion=Free
 }
 {bone1Name=dyn_jacket_bk_l_1, bone2Name=dyn_jacket_bk_l_MSTR, swing1LimitDegrees=30, swing2LimitDegrees=45, twistLimitDegrees=20}
 {bone1Name=dyn_jacket_bk_l_2, bone2Name=dyn_jacket_bk_l_1, swing1LimitDegrees=30, swing2LimitDegrees=45, twistLimitDegrees=20}
 {bone1Name=dyn_jacket_bk_l_3, bone2Name=dyn_jacket_bk_l_2, swing1LimitDegrees=30, swing2LimitDegrees=45, twistLimitDegrees=20}
 {bone1Name=dyn_jacket_bk_l_4, bone2Name=dyn_jacket_bk_l_3, swing1LimitDegrees=30, swing2LimitDegrees=45, twistLimitDegrees=20}
 {
  bone1Name=dyn_jacket_fr_r_MSTR
  bone2Name=pelvis
  swing1LimitDegrees=45
  swing1Motion=Free
  swing2LimitDegrees=45
  swing2Motion=Free
  twistLimitDegrees=45
  twistMotion=Free
 }
 {bone1Name=dyn_jacket_fr_r_1, bone2Name=dyn_jacket_fr_r_MSTR, swing1LimitDegrees=10, swing2LimitDegrees=45, twistLimitDegrees=5}
 {bone1Name=dyn_jacket_fr_r_2, bone2Name=dyn_jacket_fr_r_1, swing1LimitDegrees=10, swing2LimitDegrees=45, twistLimitDegrees=5}
 {bone1Name=dyn_jacket_fr_r_3, bone2Name=dyn_jacket_fr_r_2, swing1LimitDegrees=10, swing2LimitDegrees=45, twistLimitDegrees=5}
 {bone1Name=dyn_jacket_fr_r_4, bone2Name=dyn_jacket_fr_r_3, swing1LimitDegrees=10, swing2LimitDegrees=45, twistLimitDegrees=5}
 {
  bone1Name=dyn_jacket_mid_r_MSTR
  bone2Name=pelvis
  swing1LimitDegrees=45
  swing1Motion=Free
  swing2LimitDegrees=45
  swing2Motion=Free
  twistLimitDegrees=45
  twistMotion=Free
 }
 {bone1Name=dyn_jacket_mid_r_1, bone2Name=dyn_jacket_mid_r_MSTR, swing1LimitDegrees=10, swing2LimitDegrees=45, twistLimitDegrees=20}
 {bone1Name=dyn_jacket_mid_r_2, bone2Name=dyn_jacket_mid_r_1, swing1LimitDegrees=10, swing2LimitDegrees=45, twistLimitDegrees=20}
 {bone1Name=dyn_jacket_mid_r_3, bone2Name=dyn_jacket_mid_r_2, swing1LimitDegrees=10, swing2LimitDegrees=45, twistLimitDegrees=20}
 {bone1Name=dyn_jacket_mid_r_4, bone2Name=dyn_jacket_mid_r_3, swing1LimitDegrees=10, swing2LimitDegrees=45, twistLimitDegrees=20}
 {
  bone1Name=dyn_jacket_bk_r_MSTR
  bone2Name=pelvis
  swing1LimitDegrees=45
  swing1Motion=Free
  swing2LimitDegrees=45
  swing2Motion=Free
  twistLimitDegrees=45
  twistMotion=Free
 }
 {bone1Name=dyn_jacket_bk_r_1, bone2Name=dyn_jacket_bk_r_MSTR, swing1LimitDegrees=30, swing2LimitDegrees=45, twistLimitDegrees=20}
 {bone1Name=dyn_jacket_bk_r_2, bone2Name=dyn_jacket_bk_r_1, swing1LimitDegrees=30, swing2LimitDegrees=45, twistLimitDegrees=20}
 {bone1Name=dyn_jacket_bk_r_3, bone2Name=dyn_jacket_bk_r_2, swing1LimitDegrees=30, swing2LimitDegrees=45, twistLimitDegrees=20}
 {bone1Name=dyn_jacket_bk_r_4, bone2Name=dyn_jacket_bk_r_3, swing1LimitDegrees=30, swing2LimitDegrees=45, twistLimitDegrees=20}
 {
  bone1Name=dyn_jacket_bk_l_4
  bone2Name=dyn_jacket_mid_l_4
  swing1LimitDegrees=45
  swing1Motion=Free
  swing2LimitDegrees=45
  swing2Motion=Free
  twistLimitDegrees=45
  twistMotion=Free
 }
 {
  bone1Name=dyn_jacket_bk_r_4
  bone2Name=dyn_jacket_mid_r_4
  swing1LimitDegrees=45
  swing1Motion=Free
  swing2LimitDegrees=45
  swing2Motion=Free
  twistLimitDegrees=45
  twistMotion=Free
 }
 {
  bone1Name=dyn_jacket_bk_r_3
  bone2Name=dyn_jacket_bk_l_3
  swing1LimitDegrees=45
  swing1Motion=Free
  swing2LimitDegrees=45
  swing2Motion=Free
  twistLimitDegrees=45
  twistMotion=Free
 }
 {
  bone1Name=dyn_jacket_bk_r_2
  bone2Name=dyn_jacket_bk_l_2
  swing1LimitDegrees=45
  swing1Motion=Free
  swing2LimitDegrees=45
  swing2Motion=Free
  twistLimitDegrees=45
  twistMotion=Free
 }
]
