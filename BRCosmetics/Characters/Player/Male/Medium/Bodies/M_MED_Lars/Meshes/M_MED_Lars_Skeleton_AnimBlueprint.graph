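# Post-process style AnimBlueprint for the Lars player body cosmetic, adding secondary motion to coat, chains, necklace and accessories over the pose copied from the parent mesh.
# Weighted constraints, bone offsets and AnimDynamics build a cached PreSim pose, feeding pelvis-space RigidBody coat sims blended by skydive and crouch state.
# Button constraints, chain dynamics and parachute offsets finish the pose.
# EventGraph updates a simulation-space rotation; skydive jacket constraints sit unconnected.
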AnimBlueprint /BRCosmetics/Characters/Player/Male/Medium/Bodies/M_MED_Lars/Meshes/M_MED_Lars_Skeleton_AnimBlueprint
target_skeleton=&/M_MED_Lars_Skeleton
graphs [
 AnimGraph {nodes [
   n1 = "Constraint - dyn_jacket_fr_l_1 (P)"(Alpha=1, ConstraintWeights_0=1) {
    alphaBoolBlend {blendOption=Linear}
    boneToModify {boneName=dyn_jacket_fr_l_1}
    constraintSetup [{targetBone {boneName=pelvis}}]
    constraintWeights [1]
   }
   n2 = "Constraint - dyn_jacket_mid_l_1 (P)"(ComponentPose=@n1, Alpha=1, ConstraintWeights_0=1) {
    alphaBoolBlend {blendOption=Linear}
    boneToModify {boneName=dyn_jacket_mid_l_1}
    constraintSetup [{targetBone {boneName=pelvis}}]
    constraintWeights [1]
   }
   n3 = "Constraint - dyn_jacket_bk_l_1 (P)"(ComponentPose=@n2, Alpha=1, ConstraintWeights_0=1) {
    alphaBoolBlend {blendOption=Linear}
    boneToModify {boneName=dyn_jacket_bk_l_1}
    constraintSetup [{targetBone {boneName=pelvis}}]
    constraintWeights [1]
   }
   n4 = "Constraint - dyn_jacket_fr_r_1 (P)"(ComponentPose=@n3, Alpha=1, ConstraintWeights_0=1) {
    alphaBoolBlend {blendOption=Linear}
    boneToModify {boneName=dyn_jacket_fr_r_1}
    constraintSetup [{targetBone {boneName=pelvis}}]
    constraintWeights [1]
   }
   n5 = "Constraint - dyn_jacket_mid_r_1 (P)"(ComponentPose=@n4, Alpha=1, ConstraintWeights_0=1) {
    alphaBoolBlend {blendOption=Linear}
    boneToModify {boneName=dyn_jacket_mid_r_1}
    constraintSetup [{targetBone {boneName=pelvis}}]
    constraintWeights [1]
   }
   n6 = "Constraint - dyn_jacket_bk_r_1 (P)"(ComponentPose=@n5, Alpha=1, ConstraintWeights_0=1) {
    alphaBoolBlend {blendOption=Linear}
    boneToModify {boneName=dyn_jacket_bk_r_1}
    constraintSetup [{targetBone {boneName=pelvis}}]
    constraintWeights [1]
   }
   n7 = Get Skydive Alpha() -> (Skydive Alpha:double)
   n8 = "Constraint - dyn_jacket_fr_l_2 (P)"(ComponentPose=@n6, Alpha=@n7, ConstraintWeights_0=1) {
    alphaBoolBlend {blendOption=Linear}
    boneToModify {boneName=dyn_jacket_fr_l_2}
    constraintSetup [{targetBone {boneName=thigh_l}}]
    constraintWeights [1]
   }
   n9 = "Constraint - dyn_jacket_fr_l_3 (PP)"!(ComponentPose=@n8, Alpha=@n7, ConstraintWeights_0=0.5, ConstraintWeights_1:float=0.5) -> (Pose) {
    alphaBoolBlend {blendOption=Linear}
    boneToModify {boneName=dyn_jacket_fr_l_3}
    constraintSetup [{targetBone {boneName=thigh_l}}, {targetBone {boneName=calf_l}}]
    constraintWeights [1,1]
   }
   n10 = "Constraint - dyn_jacket_fr_l_4 (P)"(ComponentPose=@n9, Alpha=@n7, ConstraintWeights_0=1) {
    alphaBoolBlend {blendOption=Linear}
    boneToModify {boneName=dyn_jacket_fr_l_4}
    constraintSetup [{targetBone {boneName=calf_l}}]
    constraintWeights [1]
   }
   n11 = "Constraint - dyn_jacket_fr_r_2 (P)"(ComponentPose=@n10, Alpha=@n7, ConstraintWeights_0=1) {
    alphaBoolBlend {blendOption=Linear}
    boneToModify {boneName=dyn_jacket_fr_r_2}
    constraintSetup [{targetBone {boneName=thigh_r}}]
    constraintWeights [1]
   }
   n12 = "Constraint - dyn_jacket_fr_r_3 (PP)"!(ComponentPose=@n11, Alpha=@n7, ConstraintWeights_0=0.5, ConstraintWeights_1:float=0.5) -> (Pose) {
    alphaBoolBlend {blendOption=Linear}
    boneToModify {boneName=dyn_jacket_fr_r_3}
    constraintSetup [{targetBone {boneName=thigh_r}}, {targetBone {boneName=calf_r}}]
    constraintWeights [1,1]
   }
   n13 = "Constraint - dyn_jacket_fr_r_4 (P)"(ComponentPose=@n12, Alpha=@n7, ConstraintWeights_0=1) {
    alphaBoolBlend {blendOption=Linear}
    boneToModify {boneName=dyn_jacket_fr_r_4}
    constraintSetup [{targetBone {boneName=calf_r}}]
    constraintWeights [1]
   }
   n14 = "Constraint - dyn_jacket_mid_l_2 (P)"(ComponentPose=@n13, Alpha=@n7, ConstraintWeights_0=1) {
    alphaBoolBlend {blendOption=Linear}
    boneToModify {boneName=dyn_jacket_mid_l_2}
    constraintSetup [{targetBone {boneName=thigh_l}}]
    constraintWeights [1]
   }
   n15 = "Constraint - dyn_jacket_mid_l_3 (PP)"!(ComponentPose=@n14, Alpha=@n7, ConstraintWeights_0=0.5, ConstraintWeights_1:float=0.5) -> (Pose) {
    alphaBoolBlend {blendOption=Linear}
    boneToModify {boneName=dyn_jacket_mid_l_3}
    constraintSetup [{targetBone {boneName=thigh_l}}, {targetBone {boneName=calf_l}}]
    constraintWeights [1,1]
   }
   n16 = "Constraint - dyn_jacket_mid_l_4 (P)"(ComponentPose=@n15, Alpha=@n7, ConstraintWeights_0=1) {
    alphaBoolBlend {blendOption=Linear}
    boneToModify {boneName=dyn_jacket_mid_l_4}
    constraintSetup [{targetBone {boneName=calf_l}}]
    constraintWeights [1]
   }
   n17 = "Constraint - dyn_jacket_mid_r_2 (P)"(ComponentPose=@n16, Alpha=@n7, ConstraintWeights_0=1) {
    alphaBoolBlend {blendOption=Linear}
    boneToModify {boneName=dyn_jacket_mid_r_2}
    constraintSetup [{targetBone {boneName=thigh_r}}]
    constraintWeights [1]
   }
   n18 = "Constraint - dyn_jacket_mid_r_3 (PP)"!(ComponentPose=@n17, Alpha=@n7, ConstraintWeights_0=0.5, ConstraintWeights_1:float=0.5) -> (Pose) {
    alphaBoolBlend {blendOption=Linear}
    boneToModify {boneName=dyn_jacket_mid_r_3}
    constraintSetup [{targetBone {boneName=thigh_r}}, {targetBone {boneName=calf_r}}]
    constraintWeights [1,1]
   }
   n19 = "Constraint - dyn_jacket_mid_r_4 (P)"(ComponentPose=@n18, Alpha=@n7, ConstraintWeights_0=1) {
    alphaBoolBlend {blendOption=Linear}
    boneToModify {boneName=dyn_jacket_mid_r_4}
    constraintSetup [{targetBone {boneName=calf_r}}]
    constraintWeights [1]
   }
   n20 = "Constraint - dyn_jacket_bk_l_2 (P)"(ComponentPose=@n19, Alpha=@n7, ConstraintWeights_0=1) {
    alphaBoolBlend {blendOption=Linear}
    boneToModify {boneName=dyn_jacket_bk_l_2}
    constraintSetup [{targetBone {boneName=thigh_l}}]
    constraintWeights [1]
   }
   n21 = "Constraint - dyn_jacket_bk_l_3 (PP)"!(ComponentPose=@n20, Alpha=@n7, ConstraintWeights_0=0.5, ConstraintWeights_1:float=0.5) -> (Pose) {
    alphaBoolBlend {blendOption=Linear}
    boneToModify {boneName=dyn_jacket_bk_l_3}
    constraintSetup [{targetBone {boneName=thigh_l}}, {targetBone {boneName=calf_l}}]
    constraintWeights [1,1]
   }
   n22 = "Constraint - dyn_jacket_bk_l_4 (P)"(ComponentPose=@n21, Alpha=@n7, ConstraintWeights_0=1) {
    alphaBoolBlend {blendOption=Linear}
    boneToModify {boneName=dyn_jacket_bk_l_4}
    constraintSetup [{targetBone {boneName=calf_l}}]
    constraintWeights [1]
   }
   n23 = "Constraint - dyn_jacket_bk_r_2 (P)"(ComponentPose=@n22, Alpha=@n7, ConstraintWeights_0=1) {
    alphaBoolBlend {blendOption=Linear}
    boneToModify {boneName=dyn_jacket_bk_r_2}
    constraintSetup [{targetBone {boneName=thigh_r}}]
    constraintWeights [1]
   }
   n24 = "Constraint - dyn_jacket_bk_r_3 (PP)"!(ComponentPose=@n23, Alpha=@n7, ConstraintWeights_0=0.5, ConstraintWeights_1:float=0.5) -> (Pose) {
    alphaBoolBlend {blendOption=Linear}
    boneToModify {boneName=dyn_jacket_bk_r_3}
    constraintSetup [{targetBone {boneName=thigh_r}}, {targetBone {boneName=calf_r}}]
    constraintWeights [1,1]
   }
   "Constraint - dyn_jacket_bk_r_4 (P)"(ComponentPose=@n24, Alpha=@n7, ConstraintWeights_0=1) {
    alphaBoolBlend {blendOption=Linear}
    boneToModify {boneName=dyn_jacket_bk_r_4}
    constraintSetup [{targetBone {boneName=calf_r}}]
    constraintWeights [1]
   }
   n25 = "Transform (Modify) Bone - Bone: dyn_jacket_fr_l_MSTR"(Alpha=Get Left Leg Out Alpha() -> (Left Leg Out Alpha:double), Translation="0,0,3", Rotation="20,0,0", Scale="1,1,1") {
    rotationMode=BMM_Additive
    rotationSpace=BCS_BoneSpace
    translationMode=BMM_Additive
    translationSpace=BCS_BoneSpace
    alphaBoolBlend {blendOption=Linear}
    alphaScaleBiasClamp {bInterpResult=true}
    boneToModify {boneName=dyn_jacket_fr_l_MSTR}
    rotation {pitch=20}
    translation {z=3}
   }
   "Transform (Modify) Bone - Bone: dyn_jacket_fr_r_MSTR"(ComponentPose=@n25, Alpha=Get Right Leg Out Alpha() -> (Right Leg Out Alpha:double), Translation="0,0,3", Rotation="20,0,0", Scale="1,1,1") {
    rotationMode=BMM_Additive
    rotationSpace=BCS_BoneSpace
    translationMode=BMM_Additive
    translationSpace=BCS_BoneSpace
    alphaBoolBlend {blendOption=Linear}
    alphaScaleBiasClamp {bInterpResult=true}
    boneToModify {boneName=dyn_jacket_fr_r_MSTR}
    rotation {pitch=20}
    translation {z=3}
   }
   n26 = RigidBody!(ComponentPose=Local To Component(LocalPose=Use cached pose 'PreSim'()), Alpha, OverrideWorldGravity:Vector="0,0,-980", ExternalForce) -> (Pose) {
    bClampLinearTranslationLimitToRefPose=true
    bOverrideWorldGravity=true
    overridePhysicsAsset=&/M_MED_Lars_PhysicsCrouch_Coat
    simulationSpace=BaseBoneSpace
    alphaBoolBlend {blendOption=Linear}
    baseBoneRef {boneName=pelvis}
    componentAppliedLinearAccClamp=<5000 5000 5000>
    componentLinearAccScale=<0.5 0.5 0.3>
    componentLinearVelScale=<0.5 0.5 0.5>
    overrideWorldGravity {z=-980}
    simSpaceSettings {
     maxAngularAcceleration=1000
     maxAngularVelocity=3
     maxLinearAcceleration=700
     maxLinearVelocity=500
     velocityScaleZ=0
     worldAlpha=0.25
     externalLinearDragV=<0.2 0.2>
    }
   }
   n27 = "Constraint - dyn_jacket_bk_l_3 (RR)"!(ComponentPose=@n26, Alpha=0.3, ConstraintWeights_0=1, ConstraintWeights_1:float=1) -> (Pose) {
    alphaBoolBlend {blendOption=Linear}
    boneToModify {boneName=dyn_jacket_bk_l_3}
    constraintSetup [{transformType=Rotation, targetBone {boneName=dyn_jacket_mid_l_3}}, {transformType=Rotation, targetBone {boneName=dyn_jacket_bk_r_3}}]
    constraintWeights [1,1]
   }
   n28 = "Constraint - dyn_jacket_bk_l_4 (RR)"!(ComponentPose=@n27, Alpha=0.3, ConstraintWeights_0=1, ConstraintWeights_1:float=1) -> (Pose) {
    alphaBoolBlend {blendOption=Linear}
    boneToModify {boneName=dyn_jacket_bk_l_4}
    constraintSetup [{transformType=Rotation, targetBone {boneName=dyn_jacket_mid_l_4}}, {transformType=Rotation, targetBone {boneName=dyn_jacket_bk_r_4}}]
    constraintWeights [1,1]
   }
   n29 = "Constraint - dyn_chainFront_02 (P)"(ComponentPose=@n28, Alpha=0.3, ConstraintWeights_0=1) {
    alphaBoolBlend {blendOption=Linear}
    boneToModify {boneName=dyn_chainFront_02}
    constraintSetup [{targetBone {boneName=thigh_twist_01_r}}]
    constraintWeights [1]
   }
   n30 = "Constraint - dyn_chainFront_03 (P)"(ComponentPose=@n29, Alpha=0.25, ConstraintWeights_0=1) {
    alphaBoolBlend {blendOption=Linear}
    boneToModify {boneName=dyn_chainFront_03}
    constraintSetup [{targetBone {boneName=thigh_twist_01_r}}]
    constraintWeights [1]
   }
   n31 = "Constraint - thigh_twist_01_r (P)"(ComponentPose=@n30, Alpha=0.2, ConstraintWeights_0=1) {
    alphaBoolBlend {blendOption=Linear}
    boneToModify {boneName=thigh_twist_01_r}
    constraintSetup [{targetBone {boneName=dyn_chainFront_03}}]
    constraintWeights [1]
   }
   n32 = "Constraint - dyn_chainFront_05 (P)"(ComponentPose=@n31, Alpha=0.15, ConstraintWeights_0=1) {
    alphaBoolBlend {blendOption=Linear}
    boneToModify {boneName=dyn_chainFront_05}
    constraintSetup [{targetBone {boneName=thigh_twist_01_r}}]
    constraintWeights [1]
   }
   n33 = "Constraint - dyn_chainFront_06 (P)"(ComponentPose=@n32, Alpha=0.1, ConstraintWeights_0=1) {
    alphaBoolBlend {blendOption=Linear}
    boneToModify {boneName=dyn_chainFront_06}
    constraintSetup [{targetBone {boneName=thigh_twist_01_r}}]
    constraintWeights [1]
   }
   n34 = "Constraint - dyn_chainFront_07 (P)"(ComponentPose=@n33, Alpha=1, ConstraintWeights_0=1) {
    alphaBoolBlend {blendOption=Linear}
    boneToModify {boneName=dyn_chainFront_07}
    constraintSetup [{targetBone {boneName=dyn_chainBack_MSTR}}]
    constraintWeights [1]
   }
   n35 = "Transform (Modify) Bone - Bone: dyn_jacket_bk_r_MSTR"(ComponentPose=@n34, Alpha=1, Translation="0,0,1", Rotation="0,0,0", Scale="1,1,1") {
    rotationMode=BMM_Additive
    rotationSpace=BCS_BoneSpace
    translationMode=BMM_Additive
    translationSpace=BCS_BoneSpace
    alphaBoolBlend {blendOption=Linear}
    boneToModify {boneName=dyn_jacket_bk_r_MSTR}
    translation {z=1}
   }
   n36 = "Transform (Modify) Bone - Bone: dyn_jacket_mid_r_MSTR"(ComponentPose=@n35, Alpha=1, Translation="0,0,1", Rotation="0,0,0", Scale="1,1,1") {
    rotationMode=BMM_Additive
    rotationSpace=BCS_BoneSpace
    translationMode=BMM_Additive
    translationSpace=BCS_BoneSpace
    alphaBoolBlend {blendOption=Linear}
    boneToModify {boneName=dyn_jacket_mid_r_MSTR}
    translation {z=1}
   }
   n37 = "Transform (Modify) Bone - Bone: dyn_jacket_bk_l_MSTR"(ComponentPose=@n36, Alpha=1, Translation="0,0,1", Rotation="0,0,0", Scale="1,1,1") {
    rotationMode=BMM_Additive
    rotationSpace=BCS_BoneSpace
    translationMode=BMM_Additive
    translationSpace=BCS_BoneSpace
    alphaBoolBlend {blendOption=Linear}
    boneToModify {boneName=dyn_jacket_bk_l_MSTR}
    translation {z=1}
   }
   n38 = "Transform (Modify) Bone - Bone: dyn_beltBuckle_MSTR"(ComponentPose=@n37, Alpha=1, Translation="0,3,2.5", Rotation="-20,30,40", Scale="1,1,1") {
    rotationMode=BMM_Additive
    rotationSpace=BCS_BoneSpace
    translationMode=BMM_Additive
    translationSpace=BCS_BoneSpace
    alphaBoolBlend {blendOption=Linear}
    boneToModify {boneName=dyn_beltBuckle_MSTR}
    rotation=<rot -20 30 40>
    translation {y=3, z=2.5}
   }
   n39 = Anim Dynamics - Bone: dyn_chainFront_01(ComponentPose=Local To Component(LocalPose=Use cached pose 'PreSim'()), Alpha=1, ExternalForce="0,0,0") {
    alphaBoolBlend {blendOption=Linear}
    boundBone {boneName=dyn_chainFront_01}
    componentAppliedLinearAccClamp=<3000 3000 3000>
    componentLinearAccScale=<0.3 0.3 0.3>
    componentLinearVelScale=<0.1 0.1 0.1>
    physicsBodyDefinitions [{
      boundBone {boneName=dyn_chainFront_01}
      boxExtents=<10 25 25>
      constraintSetup {angularLimitsMax=<1 10 5>, angularLimitsMin=<-1 0 0>, angularTarget=<0 0 0>}
      localJointOffset=<3 0 0>
     }]
   }
   n40 = RigidBody!(ComponentPose=@n39, Alpha, OverrideWorldGravity="0,0,0", ExternalForce, ComponentLinearAccScale:Vector="0.4,0.4,0.5") -> (Pose) {
    bClampLinearTranslationLimitToRefPose=true
    bOverrideWorldGravity=true
    overridePhysicsAsset=&/M_MED_Lars_Physics_Coat
    simulationSpace=BaseBoneSpace
    alphaBoolBlend {blendOption=Linear}
    baseBoneRef {boneName=pelvis}
    componentAppliedLinearAccClamp=<5000 5000 5000>
    componentLinearAccScale=<0.25 0.25 0.4>
    componentLinearVelScale=<0.4 0.4 0.5>
    simSpaceSettings {
     maxAngularAcceleration=1000
     maxAngularVelocity=3
     maxLinearAcceleration=700
     maxLinearVelocity=500
     velocityScaleZ=0
     worldAlpha=0.35
     externalLinearDragV=<0.2 0.2>
    }
   }
   n41 = "Transform (Modify) Bone - Bone: dyn_jacket_fr_r_1"!(
    ComponentPose=@n40
    bAlphaBoolEnabled:bool=OR Boolean(A=Get bIsJumping(), B=Get bIsFalling())
    Translation="-1,0,0"
    Rotation="-5,0.000,0.000"
    Scale="1,1,1"
   ) -> (Pose) {
    alphaInputType=Bool
    rotationMode=BMM_Additive
    translationMode=BMM_Additive
    alphaBoolBlend {blendInTime=0.1, blendOption=Linear, blendOutTime=0.1}
    boneToModify {boneName=dyn_jacket_fr_r_1}
    rotation {pitch=-5}
    translation {x=-1}
   }
   n42 = Blend Poses by bool(
    BlendPose_0=Component To Local(ComponentPose=@n38)
    BlendPose_1=Component To Local(ComponentPose=@n41)
    BlendTime_0=0.5
    BlendTime_1=0.5
    bActiveValue=Get bIsCrouching()
   ) {blendType=Linear}
   n43 = RigidBody!(ComponentPose=Local To Component(LocalPose=Use cached pose 'PreSim'()), Alpha, OverrideWorldGravity="0,0,-980", ExternalForce) -> (Pose) {
    bClampLinearTranslationLimitToRefPose=true
    bOverrideWorldGravity=true
    overridePhysicsAsset=&M_MED_Lars_Physics_Coat
    simulationSpace=BaseBoneSpace
    alphaBoolBlend {blendOption=Linear}
    baseBoneRef {boneName=pelvis}
    componentAppliedLinearAccClamp=<3000 3000 3000>
    componentLinearAccScale=<0.2 0.1 0.1>
    componentLinearVelScale=<0.2 0.1 0.1>
    overrideWorldGravity {z=-980}
    simSpaceSettings {
     maxAngularAcceleration=1000
     maxAngularVelocity=3
     maxLinearAcceleration=700
     maxLinearVelocity=500
     velocityScaleZ=0
     worldAlpha=0.1
     externalLinearDragV=<0.2 0.2>
    }
   }
   n44 = "Transform (Modify) Bone - Bone: dyn_chainFront_MSTR"(ComponentPose=@n43, Alpha=1, Translation="0,1,-1", Rotation="-5.000,0.000,-0.000", Scale="1,1,1") {
    rotationMode=BMM_Additive
    rotationSpace=BCS_BoneSpace
    translationMode=BMM_Additive
    translationSpace=BCS_BoneSpace
    alphaBoolBlend {blendOption=Linear}
    boneToModify {boneName=dyn_chainFront_MSTR}
    rotation {pitch=-5}
    translation {y=1, z=-1}
   }
   n45 = "Transform (Modify) Bone - Bone: dyn_jacket_fr_r_1"!(ComponentPose=@n44, Alpha:float=1, Translation="0.000,0.000,1.5", Rotation="5,0.000,0", Scale="1,1,1") -> (Pose) {
    rotationMode=BMM_Additive
    rotationSpace=BCS_BoneSpace
    translationMode=BMM_Additive
    translationSpace=BCS_BoneSpace
    alphaBoolBlend {blendOption=Linear}
    boneToModify {boneName=dyn_jacket_fr_r_1}
    rotation {pitch=5}
    translation {z=1.5}
   }
   n46 = "Transform (Modify) Bone - Bone: dyn_jacket_mid_r_1"(ComponentPose=@n45, Alpha=1, Translation="0.000,1.5,0", Rotation="5,0,-0.000", Scale="1,1,1") {
    rotationMode=BMM_Additive
    rotationSpace=BCS_BoneSpace
    translationMode=BMM_Additive
    translationSpace=BCS_BoneSpace
    alphaBoolBlend {blendOption=Linear}
    boneToModify {boneName=dyn_jacket_mid_r_1}
    rotation {pitch=5}
    translation {y=1.5}
   }
   n47 = "Transform (Modify) Bone - Bone: dyn_jacket_bk_r_1"(ComponentPose=@n46, Alpha=1, Translation="0.000,1.5,0", Rotation="5,0,0", Scale="1,1,1") {
    rotationMode=BMM_Additive
    rotationSpace=BCS_BoneSpace
    translationMode=BMM_Additive
    translationSpace=BCS_BoneSpace
    alphaBoolBlend {blendOption=Linear}
    boneToModify {boneName=dyn_jacket_bk_r_1}
    rotation {pitch=5}
    translation {y=1.5}
   }
   n48 = "Transform (Modify) Bone - Bone: dyn_jacket_fr_l_1"(ComponentPose=@n47, Alpha=1, Translation="0.000,1.5,0", Rotation="5,0,0", Scale="1,1,1") {
    rotationMode=BMM_Additive
    rotationSpace=BCS_BoneSpace
    translationMode=BMM_Additive
    translationSpace=BCS_BoneSpace
    alphaBoolBlend {blendOption=Linear}
    boneToModify {boneName=dyn_jacket_fr_l_1}
    rotation {pitch=5}
    translation {y=1.5}
   }
   n49 = "Transform (Modify) Bone - Bone: dyn_jacket_mid_l_1"(ComponentPose=@n48, Alpha=1, Translation="0.000,1.5,0", Rotation="5,0,0", Scale="1,1,1") {
    rotationMode=BMM_Additive
    rotationSpace=BCS_BoneSpace
    translationMode=BMM_Additive
    translationSpace=BCS_BoneSpace
    alphaBoolBlend {blendOption=Linear}
    boneToModify {boneName=dyn_jacket_mid_l_1}
    rotation {pitch=5}
    translation {y=1.5}
   }
   n50 = "Transform (Modify) Bone - Bone: dyn_jacket_bk_l_1"(ComponentPose=@n49, Alpha=1, Translation="0.000,1.5,0", Rotation="5,0,0", Scale="1,1,1") {
    rotationMode=BMM_Additive
    rotationSpace=BCS_BoneSpace
    translationMode=BMM_Additive
    translationSpace=BCS_BoneSpace
    alphaBoolBlend {blendOption=Linear}
    boneToModify {boneName=dyn_jacket_bk_l_1}
    rotation {pitch=5}
    translation {y=1.5}
   }
   n51 = "Constraint - dyn_necklace_1 (P)"(ComponentPose=@n50, Alpha=0.7, ConstraintWeights_0=1) {
    alphaBoolBlend {blendOption=Linear}
    boneToModify {boneName=dyn_necklace_1}
    constraintSetup [{targetBone {boneName=spine_05}}]
    constraintWeights [1]
   }
   n52 = "Constraint - dyn_necklace_2 (P)"(ComponentPose=@n51, Alpha=0.7, ConstraintWeights_0=1) {
    alphaBoolBlend {blendOption=Linear}
    boneToModify {boneName=dyn_necklace_2}
    constraintSetup [{targetBone {boneName=spine_05}}]
    constraintWeights [1]
   }
   n53 = "Constraint - dyn_necklace_3 (P)"(ComponentPose=@n52, Alpha=0.7, ConstraintWeights_0=1) {
    alphaBoolBlend {blendOption=Linear}
    boneToModify {boneName=dyn_necklace_3}
    constraintSetup [{targetBone {boneName=spine_05}}]
    constraintWeights [1]
   }
   n54 = "Transform (Modify) Bone - Bone: dyn_collar"(ComponentPose=@n53, Alpha=1, Translation="-0.25,-0.25,0", Rotation="0,0,0", Scale="1,1,1") {
    rotationMode=BMM_Additive
    rotationSpace=BCS_BoneSpace
    scaleMode=BMM_Additive
    scaleSpace=BCS_BoneSpace
    translationMode=BMM_Additive
    translationSpace=BCS_BoneSpace
    alphaBoolBlend {blendOption=Linear}
    boneToModify {boneName=dyn_collar}
    translation=<-0.25 -0.25>
   }
   n55 = Component To Local(ComponentPose=@n54)
   n56 = Get bIsSkyDiving()
   n57 = "Constraint - dyn_button_L_MSTR (PPPPPPPPPP)"!(
    ComponentPose=Local To Component(LocalPose=Blend Poses by bool(BlendPose_0=@n55, BlendPose_1=@n42, bActiveValue=@n56) {blendType=Linear})
    Alpha=1
    ConstraintWeights_0=0.5193
    ConstraintWeights_1:float=0.0628
    ConstraintWeights_2:float=0.0113
    ConstraintWeights_3:float=0.0046
    ConstraintWeights_4:float=0.0003
    ConstraintWeights_5:float=0.0041
    ConstraintWeights_6:float=0.3689
    ConstraintWeights_7:float=0.0198
    ConstraintWeights_8:float=0.0004
    ConstraintWeights_9:float=0.0084
   ) -> (Pose) {
    alphaBoolBlend {blendOption=Linear}
    boneToModify {boneName=dyn_button_L_MSTR}
    constraintSetup [
     {targetBone {boneName=spine_01}}
     {targetBone {boneName=spine_02}}
     {targetBone {boneName=spine_03}}
     {targetBone {boneName=spine_04}}
     {targetBone {boneName=spine_05}}
     {targetBone {boneName=dyn_jacket_fr_l_1}}
     {targetBone {boneName=dyn_jacket}}
     {targetBone {boneName=dyn_jacket_bk_l_1}}
     {targetBone {boneName=dyn_jacket_bk_l_2}}
     {targetBone {boneName=dyn_jacket_bk_r_1}}
    ]
    constraintWeights [1,1,1,1,1,1,1,1,1,1]
   }
   n58 = "Constraint - dyn_button_R_MSTR (PPPPPPPPP)"!(
    ComponentPose=@n57
    Alpha=1
    ConstraintWeights_0=0.4934
    ConstraintWeights_1:float=0.0809
    ConstraintWeights_2:float=0.0197
    ConstraintWeights_3:float=0.0096
    ConstraintWeights_4:float=0.0006
    ConstraintWeights_5:float=0.004
    ConstraintWeights_6:float=0.3622
    ConstraintWeights_7:float=0.012
    ConstraintWeights_8:float=0.0177
   ) -> (Pose) {
    alphaBoolBlend {blendOption=Linear}
    boneToModify {boneName=dyn_button_R_MSTR}
    constraintSetup [
     {targetBone {boneName=spine_01}}
     {targetBone {boneName=spine_02}}
     {targetBone {boneName=spine_03}}
     {targetBone {boneName=spine_04}}
     {targetBone {boneName=spine_05}}
     {targetBone {boneName=dyn_jacket_fr_l_1}}
     {targetBone {boneName=dyn_jacket}}
     {targetBone {boneName=dyn_jacket_bk_l_1}}
     {targetBone {boneName=dyn_jacket_bk_r_1}}
    ]
    constraintWeights [1,1,1,1,1,1,1,1,1]
   }
   n59 = Anim Dynamics - Bone: dyn_chainFront_01(ComponentPose=@n58, Alpha=1, ExternalForce="0,0,0") {
    angularDampingOverride=0.9999
    bOverrideAngularDamping=true
    bUseGravityOverride=true
    simulationSpace=BoneRelative
    alphaBoolBlend {blendOption=Linear}
    boundBone {boneName=dyn_chainFront_01}
    gravityOverride {z=-980}
    physicsBodyDefinitions [{
      boundBone {boneName=dyn_chainFront_01}
      boxExtents=<10 20 20>
      constraintSetup {angularLimitsMax=<15 20 15>, angularLimitsMin=<-15 5 -15>, angularTarget=<0 0 0>}
      localJointOffset=<4 -0.2 0>
     }]
    relativeSpaceBone {boneName=root}
   }
   n60 = Anim Dynamics - Bone: dyn_chainFront_02(ComponentPose=@n59, Alpha=1, ExternalForce="0,0,0") {
    angularDampingOverride=0.9999
    bOverrideAngularDamping=true
    bUseGravityOverride=true
    simulationSpace=BoneRelative
    alphaBoolBlend {blendOption=Linear}
    boundBone {boneName=dyn_chainFront_02}
    gravityOverride {z=-980}
    physicsBodyDefinitions [{
      boundBone {boneName=dyn_chainFront_02}
      boxExtents=<10 20 20>
      constraintSetup {angularLimitsMax=<15 20 15>, angularLimitsMin=<-15 0 -15>, angularTarget=<0 0 0>}
      localJointOffset=<4 -0.2 0>
     }]
    relativeSpaceBone {boneName=root}
   }
   n61 = Anim Dynamics - Bone: dyn_chainFront_03(ComponentPose=@n60, Alpha=1, ExternalForce="0,0,0") {
    angularDampingOverride=0.9999
    bOverrideAngularDamping=true
    bUseGravityOverride=true
    simulationSpace=BoneRelative
    alphaBoolBlend {blendOption=Linear}
    boundBone {boneName=dyn_chainFront_03}
    gravityOverride {z=-980}
    physicsBodyDefinitions [{
      boundBone {boneName=dyn_chainFront_03}
      boxExtents=<10 20 20>
      constraintSetup {angularLimitsMax=<15 20 15>, angularLimitsMin=<-15 0 -15>, angularTarget=<0 0 0>}
      localJointOffset=<4 -0.2 0>
     }]
    relativeSpaceBone {boneName=root}
   }
   n62 = "Constraint - dyn_chainFront_03 (P)"(ComponentPose=@n61, Alpha=0.3, ConstraintWeights_0=1) {
    alphaBoolBlend {blendOption=Linear}
    boneToModify {boneName=dyn_chainFront_03}
    constraintSetup [{targetBone {boneName=thigh_twist_01_r}}]
    constraintWeights [1]
   }
   n63 = "Constraint - dyn_chainFront_02 (P)"(ComponentPose=@n62, Alpha=0.3, ConstraintWeights_0=1) {
    alphaBoolBlend {blendOption=Linear}
    boneToModify {boneName=dyn_chainFront_02}
    constraintSetup [{targetBone {boneName=thigh_twist_01_r}}]
    constraintWeights [1]
   }
   n64 = "Constraint - dyn_chainFront_01 (P)"(ComponentPose=@n63, Alpha=0.3, ConstraintWeights_0=1) {
    alphaBoolBlend {blendOption=Linear}
    boneToModify {boneName=dyn_chainFront_01}
    constraintSetup [{targetBone {boneName=thigh_twist_01_r}}]
    constraintWeights [1]
   }
   n65 = "Transform (Modify) Bone - Bone: dyn_chainFront_MSTR"(ComponentPose=@n64, Alpha=1, Translation="0,0,0", Rotation="-2,0.000,0", Scale="1,1,1") {
    rotationMode=BMM_Additive
    rotationSpace=BCS_BoneSpace
    translationMode=BMM_Additive
    translationSpace=BCS_BoneSpace
    alphaBoolBlend {blendOption=Linear}
    boneToModify {boneName=dyn_chainFront_MSTR}
    rotation {pitch=-2}
   }
   n66 = Get bIsParachuteOpened()
   n67 = "Transform (Modify) Bone - Bone: dyn_jacket_bk_l_MSTR"!(ComponentPose=@n65, bAlphaBoolEnabled:bool=@n66, Translation="0,0,1", Rotation="0,0,0", Scale="1,1,1") -> (Pose) {
    alphaInputType=Bool
    rotationMode=BMM_Additive
    rotationSpace=BCS_BoneSpace
    translationMode=BMM_Additive
    translationSpace=BCS_BoneSpace
    alphaBoolBlend {blendInTime=0.1, blendOption=Linear, blendOutTime=0.1}
    boneToModify {boneName=dyn_jacket_bk_l_MSTR}
    translation {z=1}
   }
   n68 = "Transform (Modify) Bone - Bone: dyn_jacket_bk_r_MSTR"!(ComponentPose=@n67, bAlphaBoolEnabled:bool=@n66, Translation="0,0,1", Rotation="0,0,0", Scale="1,1,1") -> (Pose) {
    alphaInputType=Bool
    rotationMode=BMM_Additive
    rotationSpace=BCS_BoneSpace
    translationMode=BMM_Additive
    translationSpace=BCS_BoneSpace
    alphaBoolBlend {blendInTime=0.1, blendOption=Linear, blendOutTime=0.1}
    boneToModify {boneName=dyn_jacket_bk_r_MSTR}
    translation {z=1}
   }
   Output Pose(Result=Component To Local(ComponentPose=@n68)) {layerGroup=None}
   n69 = "Constraint - dyn_walkie_MSTR (PPPPPPPPPPPPPPP)"!(
    ComponentPose=Local To Component(LocalPose=Copy Pose From Mesh() {bCopyCurves=true, bUseAttachedParent=true})
    Alpha=1
    ConstraintWeights_0=0
    ConstraintWeights_1:float=0.0001
    ConstraintWeights_2:float=0
    ConstraintWeights_3:float=0.4977
    ConstraintWeights_4:float=0.0014
    ConstraintWeights_5:float=0
    ConstraintWeights_6:float=0
    ConstraintWeights_7:float=0.0379
    ConstraintWeights_8:float=0.0215
    ConstraintWeights_9:float=0.1777
    ConstraintWeights_10:float=0.2347
    ConstraintWeights_11:float=0
    ConstraintWeights_12:float=0.0226
    ConstraintWeights_13:float=0.0064
    ConstraintWeights_14:float=0
   ) -> (Pose) {
    alphaBoolBlend {blendOption=Linear}
    boneToModify {boneName=dyn_walkie_MSTR}
    constraintSetup [
     {targetBone {boneName=dyn_jacket_fr_l_1}}
     {targetBone {boneName=head}}
     {targetBone {boneName=dyn_jacket}}
     {targetBone {boneName=clavicle_l}}
     {targetBone {boneName=dyn_lapel_l}}
     {targetBone {boneName=spine_01}}
     {targetBone {boneName=spine_02}}
     {targetBone {boneName=spine_03}}
     {targetBone {boneName=spine_04}}
     {targetBone {boneName=spine_05}}
     {targetBone {boneName=upperarm_twist_01_l}}
     {targetBone {boneName=clavicle_r}}
     {targetBone {boneName=neck_02}}
     {targetBone {boneName=neck_01}}
     {targetBone {boneName=upperarm_l}}
    ]
    constraintWeights [1,1,1,1,1,1,1,1,1,1,1,1,1,1,1]
   }
   n70 = "Transform (Modify) Bone - Bone: dyn_walkie_MSTR"(ComponentPose=@n69, Alpha=1, Translation="0,0,0", Rotation="0,-10,0", Scale="1,1,1") {
    rotationMode=BMM_Additive
    rotationSpace=BCS_BoneSpace
    scaleMode=BMM_Additive
    scaleSpace=BCS_BoneSpace
    translationMode=BMM_Additive
    translationSpace=BCS_BoneSpace
    alphaBoolBlend {blendOption=Linear}
    boneToModify {boneName=dyn_walkie_MSTR}
    rotation {yaw=-10}
   }
   n71 = "Transform (Modify) Bone - Bone: dyn_walkie_MSTR"(ComponentPose=@n70, Alpha=Get Walkie Alpha() -> (Walkie Alpha:double), Translation="1,0,-1", Rotation="0,-10,5", Scale="0.9,1,0.8") {
    rotationMode=BMM_Additive
    rotationSpace=BCS_BoneSpace
    scaleMode=BMM_Additive
    scaleSpace=BCS_BoneSpace
    translationMode=BMM_Additive
    translationSpace=BCS_BoneSpace
    alphaBoolBlend {blendOption=Linear}
    boneToModify {boneName=dyn_walkie_MSTR}
    rotation {roll=5, yaw=-10}
    scale {x=0.9, z=0.8}
    translation {x=1, z=-1}
   }
   n72 = "Transform (Modify) Bone - Bone: dyn_buttadjust"(ComponentPose=@n71, Alpha=1, Translation="0,-3.5,0", Rotation="0,0,0", Scale="1,1,1") {
    translationMode=BMM_Additive
    translationSpace=BCS_BoneSpace
    alphaBoolBlend {blendOption=Linear}
    boneToModify {boneName=dyn_buttadjust}
    translation {y=-3.5}
   }
   n73 = "Constraint - dyn_chainFront_MSTR (PPP)"!(ComponentPose=@n72, Alpha=1, ConstraintWeights_0=0.199, ConstraintWeights_1:float=0.0901, ConstraintWeights_2:float=0.7109) -> (Pose) {
    alphaBoolBlend {blendOption=Linear}
    boneToModify {boneName=dyn_chainFront_MSTR}
    constraintSetup [{targetBone {boneName=spine_01}}, {targetBone {boneName=thigh_twist_01_r}}, {targetBone {boneName=pelvis}}]
    constraintWeights [1,1,1]
   }
   n74 = "Constraint - dyn_chainBack_MSTR (PPP)"!(ComponentPose=@n73, Alpha=1, ConstraintWeights_0=0.4041, ConstraintWeights_1:float=0.4711, ConstraintWeights_2:float=0.1249) -> (Pose) {
    alphaBoolBlend {blendOption=Linear}
    boneToModify {boneName=dyn_chainBack_MSTR}
    constraintSetup [{targetBone {boneName=thigh_twist_01_r}}, {targetBone {boneName=pelvis}}, {targetBone {boneName=dyn_buttadjust}}]
    constraintWeights [1,1,1]
   }
   n75 = "Transform (Modify) Bone - Bone: dyn_chainFront_MSTR"(ComponentPose=@n74, Alpha=1, Translation="-0.5,-0.5,0", Rotation="0,0,0", Scale="1,1,1") {
    translationMode=BMM_Additive
    translationSpace=BCS_BoneSpace
    alphaBoolBlend {blendOption=Linear}
    boneToModify {boneName=dyn_chainFront_MSTR}
    translation=<-0.5 -0.5>
   }
   n76 = "Transform (Modify) Bone - Bone: dyn_chainBack_MSTR"(ComponentPose=@n75, Alpha=1, Translation="0,1,1", Rotation="0,0,0", Scale="1,1,1") {
    translationMode=BMM_Additive
    translationSpace=BCS_BoneSpace
    alphaBoolBlend {blendOption=Linear}
    boneToModify {boneName=dyn_chainBack_MSTR}
    translation {y=1, z=1}
   }
   n77 = "Constraint - dyn_collar_MSTR (PPPPPP)"!(
    ComponentPose=@n76
    Alpha=1
    ConstraintWeights_0=0.0261
    ConstraintWeights_1:float=0
    ConstraintWeights_2:float=0.0201
    ConstraintWeights_3:float=0.6938
    ConstraintWeights_4:float=0.1773
    ConstraintWeights_5:float=0.0828
   ) -> (Pose) {
    alphaBoolBlend {blendOption=Linear}
    boneToModify {boneName=dyn_collar_MSTR}
    constraintSetup [
     {targetBone {boneName=head}}
     {targetBone {boneName=dyn_collar}}
     {targetBone {boneName=spine_03}}
     {targetBone {boneName=spine_05}}
     {targetBone {boneName=neck_02}}
     {targetBone {boneName=neck_01}}
    ]
    constraintWeights [1,1,1,1,1,1]
   }
   n78 = "Transform (Modify) Bone - Bone: dyn_collar_MSTR"(
    ComponentPose=@n77
    Alpha="Get Head Rotation Back "() -> ("Head Rotation Back ":double)
    Translation="-2,-0.5,0"
    Rotation="0,-5,0"
    Scale="0.8,1,1"
   ) {
    rotationMode=BMM_Additive
    rotationSpace=BCS_BoneSpace
    scaleMode=BMM_Additive
    scaleSpace=BCS_BoneSpace
    translationMode=BMM_Additive
    translationSpace=BCS_BoneSpace
    alphaBoolBlend {blendOption=Linear}
    boneToModify {boneName=dyn_collar_MSTR}
    rotation {yaw=-5}
    scale {x=0.8}
    translation=<-2 -0.5>
   }
   n79 = "Constraint - dyn_beltBuckle_MSTR (PPPP)"!(
    ComponentPose=@n78
    Alpha=1
    ConstraintWeights_0=0.303
    ConstraintWeights_1:float=0.0124
    ConstraintWeights_2:float=0.0325
    ConstraintWeights_3:float=0.6521
   ) -> (Pose) {
    alphaBoolBlend {blendOption=Linear}
    boneToModify {boneName=dyn_beltBuckle_MSTR}
    constraintSetup [{targetBone {boneName=spine_01}}, {targetBone {boneName=spine_02}}, {targetBone {boneName=thigh_twist_01_l}}, {targetBone {boneName=pelvis}}]
    constraintWeights [1,1,1,1]
   }
   n80 = "Transform (Modify) Bone - Bone: dyn_root_simspace"(ComponentPose=@n79, Alpha=1, Translation="0,0,0", Rotation=Get sim_space_rotation() -> (sim_space_rotation:Rotator), Scale="1,1,1") {
    rotationMode=BMM_Replace
    rotationSpace=BCS_WorldSpace
    translationSpace=BCS_BoneSpace
    alphaBoolBlend {blendOption=Linear}
    alphaScaleBiasClamp {bInterpResult=true, interpSpeedDecreasing=2, interpSpeedIncreasing=2}
    boneToModify {boneName=dyn_root_simspace}
   }
   n81 = "Constraint - dyn_chainFront_MSTR (R)"(ComponentPose=@n80, Alpha=1, ConstraintWeights_0=1) {
    alphaBoolBlend {blendOption=Linear}
    boneToModify {boneName=dyn_chainFront_MSTR}
    constraintSetup [{transformType=Rotation, targetBone {boneName=thigh_r}}]
    constraintWeights [1]
   }
   n82 = "Transform (Modify) Bone - Bone: dyn_chainFront_MSTR"(ComponentPose=@n81, Alpha=Get Right Leg Alphas() -> (Right Leg Alphas:double), Translation="-1.5,0,-0.5", Rotation="-5,30,0", Scale="1,1,1") {
    rotationMode=BMM_Additive
    rotationSpace=BCS_BoneSpace
    translationMode=BMM_Additive
    translationSpace=BCS_BoneSpace
    alphaBoolBlend {blendOption=Linear}
    alphaScaleBiasClamp {bInterpResult=true}
    boneToModify {boneName=dyn_chainFront_MSTR}
    rotation {pitch=-5, yaw=30}
    translation {x=-1.5, z=-0.5}
   }
   n83 = Anim Dynamics - Bone: dyn_necklace_1(ComponentPose=@n82, Alpha=1, ExternalForce="0,0,0") {
    angularDampingOverride=0.99
    bOverrideAngularDamping=true
    bUseGravityOverride=true
    bUsePlanarLimit=false
    alphaBoolBlend {blendOption=Linear}
    boundBone {boneName=dyn_necklace_1}
    componentAppliedLinearAccClamp=<1000 1000 1000>
    componentLinearAccScale=<0.2 0.2 0.3>
    componentLinearVelScale=<0.1 0.1 0.2>
    gravityOverride {z=-800}
    physicsBodyDefinitions [{
      sphereCollisionRadius=0
      boundBone {boneName=dyn_necklace_1}
      boxExtents=<12 12 4>
      constraintSetup {angularLimitsMax=<0 0 2>, angularLimitsMin=<0 -4 -2>, angularTarget=<1 0 0>}
      localJointOffset=<2 0 0>
     }]
   }
   n84 = Anim Dynamics - Bone: dyn_necklace_2(ComponentPose=@n83, Alpha=1, ExternalForce="0,0,0") {
    angularDampingOverride=0.99
    bOverrideAngularDamping=true
    bUseGravityOverride=true
    bUsePlanarLimit=false
    simulationSpace=RootRelative
    alphaBoolBlend {blendOption=Linear}
    boundBone {boneName=dyn_necklace_2}
    componentAppliedLinearAccClamp=<1000 1000 1000>
    componentLinearAccScale=<0.2 0.2 0.3>
    componentLinearVelScale=<0.1 0.1 0.2>
    gravityOverride {z=-800}
    physicsBodyDefinitions [{
      sphereCollisionRadius=0
      boundBone {boneName=dyn_necklace_2}
      boxExtents=<14 14 4>
      constraintSetup {angularLimitsMax=<0 0 3>, angularLimitsMin=<0 -10 -3>, angularTarget=<1 0 0>}
      localJointOffset=<2 0 0>
     }]
   }
   n85 = Anim Dynamics - Bone: dyn_necklace_3(ComponentPose=@n84, Alpha=1, ExternalForce="0,0,0") {
    angularDampingOverride=0.99
    bOverrideAngularDamping=true
    bUseGravityOverride=true
    bUsePlanarLimit=false
    simulationSpace=RootRelative
    alphaBoolBlend {blendOption=Linear}
    boundBone {boneName=dyn_necklace_3}
    componentAppliedLinearAccClamp=<1000 1000 1000>
    componentLinearAccScale=<0.3 0.3 0.35>
    componentLinearVelScale=<0.1 0.1 0.2>
    gravityOverride {z=-800}
    physicsBodyDefinitions [{
      sphereCollisionRadius=0
      boundBone {boneName=dyn_necklace_3}
      boxExtents=<16 16 4>
      constraintSetup {angularLimitsMax=<0 2 5>, angularLimitsMin=<0 -15 -5>, angularTarget=<1 0 0>}
      localJointOffset=<2 0 0>
     }]
   }
   n86 = "Transform (Modify) Bone - Bone: dyn_necklace_3"!(ComponentPose=@n85, bAlphaBoolEnabled:bool=True, Translation="0,0,0", Rotation="5,0,0", Scale="1,1,1") -> (Pose) {
    alphaInputType=Bool
    rotationMode=BMM_Additive
    rotationSpace=BCS_BoneSpace
    translationMode=BMM_Additive
    translationSpace=BCS_BoneSpace
    alphaBoolBlend {blendOption=Linear}
    alphaScaleBiasClamp {bInterpResult=true}
    boneToModify {boneName=dyn_necklace_3}
    rotation {pitch=5}
   }
   n87 = Anim Dynamics - Bone: dyn_meter_1(ComponentPose=@n86, Alpha=1, ExternalForce="0,0,0") {
    angularDampingOverride=0.995
    angularSpringConstant=5000
    bAngularSpring=true
    bOverrideAngularDamping=true
    bUseGravityOverride=true
    bUsePlanarLimit=false
    simulationSpace=RootRelative
    alphaBoolBlend {blendOption=Linear}
    boundBone {boneName=dyn_meter_1}
    componentAppliedLinearAccClamp=<3000 3000 3000>
    componentLinearAccScale=<0.1 0.1 0.2>
    componentLinearVelScale=<0.1 0.1 0.2>
    gravityOverride {z=-800}
    physicsBodyDefinitions [{
      boundBone {boneName=dyn_meter_1}
      boxExtents=<15 30 30>
      constraintSetup {angularLimitsMax=<1 3 8>, angularLimitsMin=<-1 -3 -8>, angularTarget=<1 0 0>}
      localJointOffset=<2 0 0>
     }]
   }
   n88 = Anim Dynamics - Bone: dyn_meter_2(ComponentPose=@n87, Alpha=1, ExternalForce="0,0,0") {
    angularDampingOverride=0.995
    angularSpringConstant=5000
    bAngularSpring=true
    bOverrideAngularDamping=true
    bUseGravityOverride=true
    bUsePlanarLimit=false
    simulationSpace=RootRelative
    alphaBoolBlend {blendOption=Linear}
    boundBone {boneName=dyn_meter_2}
    componentAppliedLinearAccClamp=<3000 3000 3000>
    componentLinearAccScale=<0.1 0.1 0.2>
    componentLinearVelScale=<0.1 0.1 0.2>
    gravityOverride {z=-800}
    physicsBodyDefinitions [{
      boundBone {boneName=dyn_meter_2}
      boxExtents=<20 40 40>
      constraintSetup {angularLimitsMax=<1 3 8>, angularLimitsMin=<-1 -3 -8>, angularTarget=<1 0 0>}
      localJointOffset=<2 0 0>
     }]
   }
   Save cached pose 'PreSim'(Pose=Component To Local(ComponentPose=@n88)) {cacheName=PreSim}
  ]}
 EventGraph {nodes [
   Event BlueprintUpdateAnimation()
   > n1 = Sequence!(execute) -> (then_0, then_1, then_2:Exec, then_3:Exec, then_4:Exec, then_5:Exec, then_6:Exec)
   Reroute Node(InputPin=@n1.then_1)
   n2 = ComplexSimSpaceFunction DEPRECATED!(
    execute=@n1.then_5
    CharacterPartAnimInstance:Custom Character Part Anim Instance Object Reference=Self-Reference()
    InterpSpeedMax:double=9
    PawnSpeedMax:double=20
    RInterpErrorTolerance:double=1
    DeltaTime:double=0
    DebugDraw:bool=false
    sim_space_rotation_in:Rotator=Get sim_space_rotation()
    inComponentSpaceIn:bool=Get InComponentSpace() -> (InComponentSpace:bool)
    inComponentSpaceLastFrameIn:bool=Get InComponentSpaceLastFrame() -> (InComponentSpaceLastFrame:bool)
    lockedOnSimSpaceIn:bool=Get LockedOnSimSpace() -> (LockedOnSimSpace:bool)
    InterpSpeedIn:double=Get InterpSpeed() -> (InterpSpeed:double)
   ) -> (
    then
    sim_space_rotation:Rotator
    inComponentSpaceOut:bool
    inComponentSpaceLastFrameOut:bool
    lockedOnSimSpaceOut:bool
    InterpSpeedOut:double
   )
   > Set sim_space_rotation(sim_space_rotation:Rotator=@n2) -> (Output_Get:Rotator)
   > Set InComponentSpace(InComponentSpace:bool=@n2.inComponentSpaceOut) -> (Output_Get:bool)
   > Set InComponentSpaceLastFrame(InComponentSpaceLastFrame:bool=@n2.inComponentSpaceLastFrameOut) -> (Output_Get:bool)
   > Set LockedOnSimSpace(LockedOnSimSpace:bool=@n2.lockedOnSimSpaceOut) -> (Output_Get:bool)
   > Set InterpSpeed(InterpSpeed:double=@n2.InterpSpeedOut) -> (Output_Get:double)
   n3 = Get bIsAccelerating2D()
   n4 = Branch(execute=@n1.then_6, Condition=AND Boolean(A=Get bIsCrouching(), B=NOT Boolean(A=@n3)))
   > Set StaticCrouch(StaticCrouch:bool=true) -> (Output_Get:bool)
   Set StaticCrouch(execute=@n4.else, StaticCrouch=false)
   n5 = GetOwnerSkeletalMeshComponent()
   n6 = BreakRotator(InRot=BreakTransform(InTransform=GetDeltaTransformFromRefPose(self=@n5, BoneName=thigh_l, BaseName=pelvis)).Rotation)
   Set Left Leg Alphas(execute=@n1.then_3, Left Leg Alphas:double=MapRangeClamped(Value=@n6.Yaw, InRangeA=-20, InRangeB=-100)) -> (Output_Get:double)
   > Set Left Leg Out Alpha(Left Leg Out Alpha:double=MapRangeClamped(Value=@n6.Yaw, InRangeA=5, InRangeB=25)) -> (Output_Get:double)
   n7 = BreakTransform(InTransform=GetDeltaTransformFromRefPose(self=GetOwnerSkeletalMeshComponent(), BoneName=upperarm_l, BaseName=spine_05))
   > Set Walkie Alpha(Walkie Alpha:double=MapRangeClamped(Value=BreakRotator(InRot=@n7.Rotation).Yaw, InRangeA=60, InRangeB=85)) -> (Output_Get:double)
   n8 = GetOwnerSkeletalMeshComponent()
   n9 = BreakRotator(InRot=BreakTransform(InTransform=GetDeltaTransformFromRefPose(self=@n8, BoneName=thigh_r, BaseName=pelvis)).Rotation)
   > Set Right Leg Alphas(Right Leg Alphas:double=MapRangeClamped(Value=@n9.Yaw, InRangeA=-30, InRangeB=-90)) -> (Output_Get:double)
   > Set Right Leg Out Alpha(Right Leg Out Alpha:double=MapRangeClamped(Value=@n9.Pitch, InRangeA=5, InRangeB=25)) -> (Output_Get:double)
   n10 = BreakTransform(InTransform=GetDeltaTransformFromRefPose(self=GetOwnerSkeletalMeshComponent(), BoneName=head, BaseName=spine_05))
   "Set Head Rotation Back "(execute=@n1.then_2, "Head Rotation Back ":double=MapRangeClamped(Value=BreakRotator(InRot=@n10.Rotation).Yaw, InRangeA=10, InRangeB=-40)) -> (Output_Get:double)
   n11 = Gravity Override State Function DEPRECATED(
    execute=@n1.then_4
    CharacterPartAnimInstance=Self-Reference()
    GameTimeInSeconds=GetGameTimeInSeconds()
    GravityOverrideParams=Get Gravity Override Params()
   )
   > Set Gravity Override(Gravity Override=@n11)
   > n12 = Branch(Condition=Get bIsSkyDiving())
   > Set Skydiving Vector Force(Skydiving Vector Force:Vector=Break Fort Anim Input Skydiving External Force(FortAnimInput_SkydivingExternalForce=Get SkydivingExternalForceInput()) -> (
     bUseSkydivingVectorForce=~
     SkydivingVectorForce=~
     HeadToPelvisDirection=~
     FloatingMultiplier=~
     FloatingAdditive=~
     DivingMultiplier=~
     DivingAdditive=~
     ParachutingMultiplier=~
     ParachutingAdditive=~
    ).SkydivingVectorForce) -> (Output_Get:Vector)
   Set Skydiving Vector Force(execute=@n12.else, Skydiving Vector Force="0, 0, 0")
   Event BlueprintInitializeAnimation()
   > Delay(Duration=0.2)
   > ResetDynamics(InTeleportType=TeleportPhysics)
   > ResetDynamics(InTeleportType=ResetPhysics)
   n13 = Self-Reference()
   n14 = BreakTransform(InTransform=GetDeltaTransformFromRefPose(self=GetPartSkeletalMeshComponent(self=@n13), BoneName=dyn_jacket_bk_r_2, BaseName=dyn_jacket_bk_r_1))
   Set Chain Back Tr Alpha(Chain Back Tr Alpha:double=MapRangeClamped(Value=BreakRotator(InRot=@n14.Rotation).Yaw, InRangeA=-5, InRangeB=2)) -> (Output_Get:double)
   n15 = Self-Reference()
   n16 = BreakTransform(InTransform=GetDeltaTransformFromRefPose(self=GetPartSkeletalMeshComponent(self=@n15), BoneName=dyn_jacket_mid_r_2, BaseName=dyn_jacket_mid_r_1))
   Set Chain Side Tr Alpha(Chain Side Tr Alpha:double=MapRangeClamped(Value=BreakRotator(InRot=@n16.Rotation).Pitch, InRangeA=10, InRangeB=5)) -> (Output_Get:double)
  ]}
]
variables [
 Chain Back Tr Alpha {type=number}
 Chain Side Tr Alpha {type=number}
 Gravity Override {type=Vector}
 Gravity Override Params {
  type=GravityOverrideParamsStruct
  default {skydive_FloatingDown {gravityOverride=<0 -700 300>}, skydive_GoingUp {gravityOverride=<0 -300 500>}}
 }
 "Head Rotation Back " {type=number}
 InComponentSpace {type=boolean}
 InComponentSpaceLastFrame {type=boolean}
 InterpSpeed {type=number}
 Left Leg Alphas {type=number}
 Left Leg Out Alpha {type=number}
 LockedOnSimSpace {type=boolean}
 Right Leg Alphas {type=number}
 Right Leg Out Alpha {type=number}
 sim_space_rotation {type=Rotator}
 Skydive Alpha {type=number, default=0.5}
 Skydiving Vector Force {type=Vector}
 StaticCrouch {type=boolean}
 Walkie Alpha {type=number}
]
