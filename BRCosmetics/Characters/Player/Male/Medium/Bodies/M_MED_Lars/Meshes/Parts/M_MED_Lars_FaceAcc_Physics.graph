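# Physics asset for the Lars face accessory mesh, defining capsule bodies and joints for dynamic hair and earrings around the head and neck.
# Some hair bodies hang from back, left and right master bones, while the other hair and both earrings joint directly to the head.
# Hair and earring joints limit swing with very tight twist, while master-to-head and head-to-neck joints leave rotation free.
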PhysicsAsset /BRCosmetics/Characters/Player/Male/Medium/Bodies/M_MED_Lars/Meshes/Parts/M_MED_Lars_FaceAcc_Physics
bodies {
 dyn_earring_l [{shapeType=Capsule, length=1, radius=1, shapeName=dyn_earring_l_capsule, center=<1.3644 0 0>, rotation=<rot 90 0 0>}]
 dyn_earring_r [{shapeType=Capsule, length=1, radius=1, shapeName=dyn_earring_r_capsule, center=<1.3644 0 0>, rotation=<rot 90 0 0>}]
 dyn_hair_01 [{shapeType=Capsule, length=1, radius=1, shapeName=dyn_hair_01_capsule, center=<1.7074 0 0>, rotation=<rot 90 0 0>}]
 dyn_hair_02 [{shapeType=Capsule, length=1, radius=1, shapeName=dyn_hair_02_capsule, center=<1.343 0 0>, rotation=<rot 90 0 0>}]
 dyn_hair_03 [{shapeType=Capsule, length=1, radius=1, shapeName=dyn_hair_03_capsule, center=<0.6482 0 0>, rotation=<rot 90 0 0>}]
 dyn_hair_04 [{shapeType=Capsule, length=1.125, radius=1.625, shapeName=dyn_hair_04_capsule, center=<4.1369 0 0>, rotation=<rot 90 0 0>}]
 dyn_hair_05 [{shapeType=Capsule, length=1.125, radius=1.625, shapeName=dyn_hair_05_capsule, center=<4.1369 0 0>, rotation=<rot 90 0 0>}]
 dyn_hair_06 [{shapeType=Capsule, length=0.75, radius=0.75, shapeName=dyn_hair_06_capsule, center=<0.9028 0 0>, rotation=<rot 90 0 0>}]
 dyn_hair_07 [{shapeType=Capsule, length=1, radius=1.25, shapeName=dyn_hair_07_capsule, center=<0.7827 0 0>, rotation=<rot 90 0 0>}]
 dyn_hair_08 [{shapeType=Capsule, length=1, radius=1, shapeName=dyn_hair_08_capsule, center=<1.3644 0 0>, rotation=<rot 90 0 0>}]
 dyn_hair_09 [{shapeType=Capsule, length=1.75, radius=1.125, shapeName=dyn_hair_09_capsule, center=<1.3644 0 0>, rotation=<rot 90 0 0>}]
 dyn_hair_10 [{shapeType=Capsule, length=1, radius=1, shapeName=dyn_hair_10_capsule, center=<1.3644 0 0>, rotation=<rot 90 0 0>}]
 dyn_hair_BK_MSTR [{shapeType=Capsule, length=0.505, radius=0.505, shapeName=dyn_hair_BK_MSTR_capsule, rotation=<rot 0 -90 0>}]
 dyn_hair_L_MSTR [{shapeType=Capsule, length=0.505, radius=0.505, shapeName=dyn_hair_L_MSTR_capsule, rotation=<rot 0 -90 0>}]
 dyn_hair_R_MSTR [{shapeType=Capsule, length=0.505, radius=0.505, shapeName=dyn_hair_R_MSTR_capsule, rotation=<rot 0 -90 0>}]
 head [{shapeType=Capsule, length=5.6276, radius=6.2393, center=<2.0768 1.8487 0>, rotation=<rot 90 -90 -90.9517>}]
 neck_02 [{shapeType=Capsule, length=1.255, radius=4.005, shapeName=neck_02_capsule, rotation=<rot 90 0 0>}]
}
constraints [
 {bone1Name=dyn_earring_r, bone2Name=head, swing1LimitDegrees=35, swing2LimitDegrees=35, twistLimitDegrees=3}
 {bone1Name=dyn_earring_l, bone2Name=head, swing1LimitDegrees=35, swing2LimitDegrees=35, twistLimitDegrees=3}
 {bone1Name=dyn_hair_01, bone2Name=dyn_hair_BK_MSTR, swing1LimitDegrees=45, swing2LimitDegrees=45, twistLimitDegrees=1}
 {bone1Name=dyn_hair_02, bone2Name=dyn_hair_BK_MSTR, swing1LimitDegrees=45, swing2LimitDegrees=45, twistLimitDegrees=1}
 {bone1Name=dyn_hair_03, bone2Name=dyn_hair_BK_MSTR, swing1LimitDegrees=45, swing2LimitDegrees=45, twistLimitDegrees=1}
 {bone1Name=dyn_hair_04, bone2Name=dyn_hair_L_MSTR, swing1LimitDegrees=45, swing2LimitDegrees=45, twistLimitDegrees=1}
 {bone1Name=dyn_hair_05, bone2Name=dyn_hair_R_MSTR, swing1LimitDegrees=45, swing2LimitDegrees=45, twistLimitDegrees=1}
 {bone1Name=dyn_hair_06, bone2Name=dyn_hair_BK_MSTR, swing1LimitDegrees=45, swing2LimitDegrees=5, twistLimitDegrees=1}
 {bone1Name=dyn_hair_07, bone2Name=head, swing1LimitDegrees=45, swing2LimitDegrees=45, twistLimitDegrees=1}
 {bone1Name=dyn_hair_08, bone2Name=head, swing1LimitDegrees=45, swing2LimitDegrees=45, twistLimitDegrees=1}
 {bone1Name=dyn_hair_09, bone2Name=head, swing1LimitDegrees=10, swing2LimitDegrees=5, twistLimitDegrees=1}
 {bone1Name=dyn_hair_10, bone2Name=head, swing1LimitDegrees=10, swing2LimitDegrees=5, twistLimitDegrees=1}
 {
  bone1Name=dyn_hair_BK_MSTR
  bone2Name=head
  swing1LimitDegrees=45
  swing1Motion=Free
  swing2LimitDegrees=45
  swing2Motion=Free
  twistLimitDegrees=45
  twistMotion=Free
 }
 {
  bone1Name=dyn_hair_R_MSTR
  bone2Name=head
  swing1LimitDegrees=45
  swing1Motion=Free
  swing2LimitDegrees=45
  swing2Motion=Free
  twistLimitDegrees=45
  twistMotion=Free
 }
 {
  bone1Name=dyn_hair_L_MSTR
  bone2Name=head
  swing1LimitDegrees=45
  swing1Motion=Free
  swing2LimitDegrees=45
  swing2Motion=Free
  twistLimitDegrees=45
  twistMotion=Free
 }
 {
  bone1Name=head
  bone2Name=neck_02
  swing1LimitDegrees=45
  swing1Motion=Free
  swing2LimitDegrees=45
  swing2Motion=Free
  twistLimitDegrees=45
  twistMotion=Free
 }
]
