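# Post-copy animation Blueprint for the Lars face accessory mesh, driving hair motion on the male medium character.
# It copies the attached parent's pose, then additively offsets back, left and right hair master bones before a head-relative RigidBody simulation with custom gravity.
# Each update, the event graph maps head rotation relative to spine_05 into clamped alphas for those offsets, shifting the hair as the head turns.
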AnimBlueprint /BRCosmetics/Characters/Player/Male/Medium/Bodies/M_MED_Lars/Meshes/Parts/M_MED_Lars_FaceAcc_AnimBP
target_skeleton=&/M_MED_Lars_FaceAcc_Skeleton
graphs [
 AnimGraph {nodes [
   n1 = "Transform (Modify) Bone - Bone: dyn_hair_BK_MSTR"(
    ComponentPose=Local To Component(LocalPose=Copy Pose From Mesh() {bUseAttachedParent=true})
    Alpha=Get Head Back Rotation() -> (Head Back Rotation:double)
    Translation="-1,1.75,0"
    Rotation="0,2,0"
    Scale="1,1,1"
   ) {
    rotationMode=BMM_Additive
    rotationSpace=BCS_BoneSpace
    scaleSpace=BCS_BoneSpace
    translationMode=BMM_Additive
    translationSpace=BCS_BoneSpace
    alphaBoolBlend {blendOption=Linear}
    boneToModify {boneName=dyn_hair_BK_MSTR}
    rotation {yaw=2}
    translation=<-1 1.75>
   }
   n2 = "Transform (Modify) Bone - Bone: dyn_hair_L_MSTR"(ComponentPose=@n1, Alpha=Get Head Side Rotation() -> (Head Side Rotation:double), Translation="0,2,0", Rotation="0,0,0", Scale="1,1,1") {
    rotationMode=BMM_Additive
    rotationSpace=BCS_BoneSpace
    scaleSpace=BCS_BoneSpace
    translationMode=BMM_Additive
    translationSpace=BCS_BoneSpace
    alphaBoolBlend {blendOption=Linear}
    boneToModify {boneName=dyn_hair_L_MSTR}
    translation {y=2}
   }
   n3 = Get Left Hair Scoot() -> (Left Hair Scoot:double)
   n4 = "Transform (Modify) Bone - Bone: dyn_hair_L_MSTR"(ComponentPose=@n2, Alpha=@n3, Translation="2,2,-2", Rotation="0,45,45", Scale="1,1,1") {
    rotationMode=BMM_Additive
    rotationSpace=BCS_BoneSpace
    scaleSpace=BCS_BoneSpace
    translationMode=BMM_Additive
    translationSpace=BCS_BoneSpace
    alphaBoolBlend {blendOption=Linear}
    boneToModify {boneName=dyn_hair_L_MSTR}
    rotation {roll=45, yaw=45}
    translation=<2 2 -2>
   }
   n5 = "Transform (Modify) Bone - Bone: dyn_hair_BK_MSTR"(ComponentPose=@n4, Alpha=@n3, Translation="0,1,0", Rotation="0,0,0", Scale="1,1,1") {
    rotationMode=BMM_Additive
    rotationSpace=BCS_BoneSpace
    scaleSpace=BCS_BoneSpace
    translationMode=BMM_Additive
    translationSpace=BCS_BoneSpace
    alphaBoolBlend {blendOption=Linear}
    boneToModify {boneName=dyn_hair_BK_MSTR}
    translation {y=1}
   }
   n6 = "Transform (Modify) Bone - Bone: dyn_hair_R_MSTR"(ComponentPose=@n5, Alpha=Get Right Hair Scoot() -> (Right Hair Scoot:double), Translation="2,0,2", Rotation="0,45,-35", Scale="1,1,1") {
    rotationMode=BMM_Additive
    rotationSpace=BCS_BoneSpace
    scaleSpace=BCS_BoneSpace
    translationMode=BMM_Additive
    translationSpace=BCS_BoneSpace
    alphaBoolBlend {blendOption=Linear}
    boneToModify {boneName=dyn_hair_R_MSTR}
    rotation {roll=-35, yaw=45}
    translation {x=2, z=2}
   }
   n7 = RigidBody(ComponentPose=@n6) {
    bClampLinearTranslationLimitToRefPose=true
    bOverrideWorldGravity=true
    overridePhysicsAsset=&/M_MED_Lars_FaceAcc_Physics
    simulationSpace=BaseBoneSpace
    alphaBoolBlend {blendOption=Linear}
    baseBoneRef {boneName=head}
    componentAppliedLinearAccClamp=<5000 5000 5000>
    componentLinearAccScale=<0.2 0.2 0.4>
    componentLinearVelScale=<0.1 0.1 0.2>
    overrideWorldGravity {z=-450}
   }
   Output Pose(Result=Component To Local(ComponentPose=@n7)) {layerGroup=None}
  ]}
 EventGraph {nodes [
   Event BlueprintUpdateAnimation()
   > n1 = Sequence()
   n2 = GetOwnerSkeletalMeshComponent()
   n3 = BreakRotator(InRot=BreakTransform(InTransform=GetDeltaTransformFromRefPose(self=@n2, BoneName=head, BaseName=spine_05)).Rotation)
   > Set Head Back Rotation(Head Back Rotation:double=MapRangeClamped(Value=@n3.Yaw, InRangeA=8, InRangeB=-28)) -> (Output_Get:double)
   > Set Head Side Rotation(Head Side Rotation:double=MapRangeClamped(Value=@n3, InRangeA=10, InRangeB=45)) -> (Output_Get:double)
   n4 = GetOwnerSkeletalMeshComponent()
   n5 = BreakRotator(InRot=BreakTransform(InTransform=GetDeltaTransformFromRefPose(self=@n4, BoneName=head, BaseName=spine_05)).Rotation)
   Set Right Hair Scoot(execute=@n1.then_1, Right Hair Scoot:double=MapRangeClamped(Value=@n5, InRangeA=-35, InRangeB=-65)) -> (Output_Get:double)
   > Set Left Hair Scoot(Left Hair Scoot:double=MapRangeClamped(Value=@n5, InRangeA=55, InRangeB=85)) -> (Output_Get:double)
  ]}
]
variables [
 Head Back Rotation {type=number}
 Head Side Rotation {type=number}
 Left Hair Scoot {type=number}
 Right Hair Scoot {type=number}
]
